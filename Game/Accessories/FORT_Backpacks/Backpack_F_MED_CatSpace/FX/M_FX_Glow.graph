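# Additive unlit glow material for the CatSpace backpack FX, usable on Niagara sprites and mesh particles.
# Color blends two parameters across a smoothstepped, rotatable UV gradient, then multiplies by particle color and feeds MF_FX_EmissiveAdjustments.
# A separate mask multiplies a procedural radial falloff by view-angle fade, particle alpha and an optional depth fade.
# No textures are sampled; the glow shape comes entirely from UV math.
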Material /Game/Accessories/FORT_Backpacks/Backpack_F_MED_CatSpace/FX/M_FX_Glow
graph {
 properties {
  bUsedWithNiagaraMeshParticles=true
  bUsedWithNiagaraSprites=true
  blendMode=BLEND_Additive
  pixelDepthOffsetMode=PDOM_Legacy
  shadingModel=MSM_Unlit
  usedShadingModels=Unlit
 }
 nodes [
  n1 = VectorParameter() {group=Color, parameterName=Color 1, defaultValue=<rgba 1.5 0.0024 1.2786 1>}
  n2 = VectorParameter() {group=Color, parameterName=Color 2, defaultValue=<rgba 0.1195 0 0.402 1>}
  n3 = ScalarParameter() {defaultValue=0.3, parameterName=Mask Position}
  n4 = MaterialFunctionCall(Float2=Rotator(Coordinate=TextureCoordinate(), Time=ScalarParameter() {defaultValue=1, parameterName=Rot})) {materialFunction=&BreakOutFloat2Components}
  n5 = SmoothStep(Min=@n3, Max=Add(A=@n3, B=Constant() {r=0.3}), Value=@n4.R) {constMax=0.5, constMin=0.5}
  n6 = ParticleColor()
  MaterialFunctionCall(Emissive=Multiply(A=LinearInterpolate(A=@n1.RGB, B=@n2.RGB, Alpha=@n5), B=@n6.RGB)) {materialFunction=&MF_FX_EmissiveAdjustments}
  n7 = Multiply(A=Distance(A=TextureCoordinate(), B=Constant() {r=0.5})) {constB=2}
  n8 = Power(Base=OneMinus(None=Saturate(None=@n7)), Exp=ScalarParameter() {defaultValue=2, parameterName=Falloff Exponent})
  n9 = MaterialFunctionCall("FadeSensitivity (optional)"=ScalarParameter() {defaultValue=1.5, parameterName=Angle Fade}) {materialFunction=&FadeBasedOnViewAngle}
  n10 = ScalarParameter() {group=DepthFade, parameterName=FadeDistance}
  Multiply(
   A=Multiply(A=Multiply(A=@n8, B=@n9.Result), B=@n6.A)
   B=LinearInterpolate(B=DepthFade(FadeDistance=@n10), Alpha=ScalarParameter() {group=DepthFade, parameterName=UseDepthFade?}) {constA=1}
  )
  OneMinus(None=@n4.R)
 ]
}
